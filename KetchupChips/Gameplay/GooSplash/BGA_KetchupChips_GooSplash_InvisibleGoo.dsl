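; Short-lived invisible goo actor for the KetchupChips GooSplash, which applies a gameplay effect to player pawns that overlap it.
; BeginPlay starts a timer that destroys the actor after half a second.
; On overlap it casts to FortPlayerPawn, stores the pawn, and uses its own ability system component to apply the KetchupChips base effect to that pawn's component.
; Tick is empty, and a separate DestroyGooPool function also destroys the actor.
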(event EventBeginPlay
  (|CallParentFunction)
  (Utilities|Time|SetTimerbyEvent (delegate Custom|DestroyGoo) 0.5))

(event Custom|DestroyGoo
  (Actor|DestroyActor) ; Kills the spaceship once\nthe timeline is complete.
)

(event EventTick
  (params (DeltaSeconds :type Float%20%28single-precision%29)))

(event Collision|EventActorBeginOverlap
  (params (OtherActor :type Actor%20Object%20Reference))
  (|CallParentFunction OtherActor)
  (bind _asfort_player_pawn (Utilities|Casting|CastToFortPlayerPawn OtherActor))
  (bind _output_get (Variables|Default|SetPlayer%20Pawn _asfort_player_pawn))
  (Utilities|IsValid (Ability|GetAbilitySystemComponent _output_get)
    (:Is%20Valid
      (GameplayEffects|ApplyGameplayEffectToTarget :self (Ability|GetAbilitySystemComponent self) :GameplayEffectClass (Variables|Default|GetKetchupChips%20Base%20Effects) :Target (Ability|GetAbilitySystemComponent _output_get) :Level 0.0))))

(fn DestroyGooPool
  (Actor|DestroyActor))

(fn ConstructionScript)
