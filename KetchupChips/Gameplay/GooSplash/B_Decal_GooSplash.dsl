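; Decal actor Blueprint for a goo splash impact mark in the KetchupChips gameplay content.
; Its construction script creates a dynamic material instance from the ViviBrondo ImpactCaustic decal material.
; It stores that instance in the Decal_MID variable and applies it to the Decal component.
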(fn ConstructionScript
  (bind _returnvalue (Rendering|Material|CreateDynamicMaterialInstance "/VividRazorConsumablesGameplay/Assets/FX/MI_Decal_ViviBrondo_ImpactCaustic.MI_Decal_ViviBrondo_ImpactCaustic" "None" "None" :!class /Script/Engine.KismetMaterialLibrary))
  (bind _output_get (Variables|Default|SetDecal_MID _returnvalue))
  (Rendering|Components|Decal|SetDecalMaterial (Variables|Decal|GetDecal) _output_get))
